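# Physics asset for the front face accessory mesh of the M_MED_CombCrater male medium body, covering two dynamic front hair chains.
# Each left and right chain runs from a small base sphere through two capsules to a box tip, with tight, slightly asymmetric swing and twist limits.
# Both bases link to neck_01 with free angular motion.
# Unconstrained head, neck, upper spine, clavicle and upper-arm capsules provide body collision volumes around the chains.
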PhysicsAsset /BRCosmetics/Characters/Player/Male/Medium/Bodies/M_MED_CombCrater/Meshes/Parts/M_MED_CombCrater_Front_FaceAcc_Physics
bodies {
 clavicle_l [{shapeType=Capsule, length=11, radius=9, shapeName=clavicle_l_capsule, center=<6.9318 -3.3714 2.5089>, rotation=<rot 82.2426 -0.0677 -2.2974>}]
 clavicle_r [{
   shapeType=Capsule
   length=11.5
   radius=9
   shapeName=clavicle_r_capsule
   center=<-6.8993 3.0744 -2.0763>
   rotation=<rot -77.8196 179.9989 -172.5332>
  }]
 dyn_hair_fr_l_01 [{
   shapeType=Capsule
   length=1.5
   radius=3.5
   shapeName=dyn_hair_fr_l_01_capsule
   center=<7.5131 -0.4201 -1.3682>
   rotation=<rot 64.4045 -39.6086 32.488>
  }]
 dyn_hair_fr_l_02 [{shapeType=Capsule, length=5, radius=1.5, shapeName=dyn_hair_fr_l_02_capsule, center=<3.9149 0 0.2374>, rotation=<rot 0 0 90>}]
 dyn_hair_fr_l_03 [{shapeType=Box, extentX=7.5, extentY=9.5, extentZ=2.5, shapeName=dyn_hair_fr_l_03_box, center=<2.8136 0 0>}]
 dyn_hair_fr_r_01 [{
   shapeType=Capsule
   length=1.5
   radius=3.5
   shapeName=dyn_hair_fr_l_01_capsule
   center=<6.5679 1.7673 -0.4158>
   rotation=<rot 78.1027 -65.3565 33.4252>
  }]
 dyn_hair_fr_r_02 [{
   shapeType=Capsule
   length=5.5
   radius=2
   shapeName=dyn_hair_fr_r_02_capsule
   center=<4.3054 0.0382 0.185>
   rotation=<rot 75.3551 179.9918 89.4721>
  }]
 dyn_hair_fr_r_03 [{shapeType=Box, extentX=7.5, extentY=9.5, extentZ=2.5, shapeName=dyn_hair_fr_r_03_box, center=<3.3162 0 -0.0018>}]
 hair_fr_l_base [{shapeType=Sphere, radius=0.505, shapeName=hair_fr_l_base_sphere}]
 hair_fr_r_base [{shapeType=Sphere, radius=0.505, shapeName=hair_fr_r_base_sphere}]
 head [
  {shapeType=Capsule, length=9.5, radius=6.5, shapeName=head_capsule, center=<2.4617 3.4932 0>, rotation=<rot 90 0 0>}
  {shapeType=Capsule, length=58, radius=30, shapeName=head_capsule, center=<34.2881 12.9113 0>}
 ]
 neck_01 [{shapeType=Capsule, length=10.5, radius=5.5, shapeName=neck_01_capsule, rotation=<rot 90 0 0>}]
 spine_05 [
  {shapeType=Capsule, length=1, radius=29, shapeName=spine_05_capsule, center=<17.9358 -34.0484 0>}
  {shapeType=Capsule, length=16, radius=25.5, shapeName=spine_05_capsule, center=<-11.1979 -9.2799 0>}
 ]
 upperarm_l [{shapeType=Capsule, length=17, radius=8, shapeName=upperarm_l_capsule, center=<7.61 -1.609 0.3605>, rotation=<rot 82.2742 38.387 48.3601>}]
 upperarm_r [{shapeType=Capsule, length=27.5, radius=4.5, shapeName=upperarm_r_capsule, center=<-13 0 0>, rotation=<rot -90 0 0>}]
}
constraints [
 {bone1Name=dyn_hair_fr_l_02, bone2Name=dyn_hair_fr_l_01, swing1LimitDegrees=20, swing2LimitDegrees=40, twistLimitDegrees=15}
 {bone1Name=dyn_hair_fr_l_03, bone2Name=dyn_hair_fr_l_02, swing1LimitDegrees=10, swing2LimitDegrees=40, twistLimitDegrees=8}
 {bone1Name=dyn_hair_fr_r_02, bone2Name=dyn_hair_fr_r_01, swing1LimitDegrees=10, swing2LimitDegrees=40, twistLimitDegrees=20}
 {bone1Name=dyn_hair_fr_r_03, bone2Name=dyn_hair_fr_r_02, swing1LimitDegrees=20, swing2LimitDegrees=40, twistLimitDegrees=8}
 {bone1Name=dyn_hair_fr_r_01, bone2Name=hair_fr_r_base, swing1LimitDegrees=10, swing2LimitDegrees=40, twistLimitDegrees=1}
 {
  bone1Name=hair_fr_r_base
  bone2Name=neck_01
  swing1LimitDegrees=45
  swing1Motion=Free
  swing2LimitDegrees=45
  swing2Motion=Free
  twistLimitDegrees=45
  twistMotion=Free
 }
 {bone1Name=dyn_hair_fr_l_01, bone2Name=hair_fr_l_base, swing1LimitDegrees=10, swing2LimitDegrees=40, twistLimitDegrees=5}
 {
  bone1Name=hair_fr_l_base
  bone2Name=neck_01
  swing1LimitDegrees=45
  swing1Motion=Free
  swing2LimitDegrees=45
  swing2Motion=Free
  twistLimitDegrees=45
  twistMotion=Free
 }
]
