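# Physics asset for the CombCraterHat character's Backbling FaceAcc part, simulating dangling back, side and front hair-bone chains.
# Capsules cover head, neck, upper spine, clavicles and upper arms, with an oversized offset second capsule on spine_05.
# Back and side chains hang from the head with swing/twist limits; front chains end in boxes, their bases angularly free on neck_01.
# Free-angular cross-constraints tie neighboring chains together, back chains to spine_05, and only the left front base to head.
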PhysicsAsset /BRCosmetics/Characters/Player/Male/Medium/Bodies/M_MED_CombCraterHat/Meshes/Parts/M_MED_CombCraterHat_Backbling_FaceAcc_Physics
bodies {
 clavicle_l [{shapeType=Capsule, length=10, radius=5, shapeName=clavicle_l_capsule, center=<9.6963 -2.0347 4.8516>, rotation=<rot 65.0051 0.5328 0.5944>}]
 clavicle_r [{
   shapeType=Capsule
   length=9.5
   radius=6
   shapeName=clavicle_r_capsule
   center=<-8.9706 0.9338 -3.6212>
   rotation=<rot -75.0043 -179.067 179.0275>
  }]
 dyn_hair_bck_l_01 [{shapeType=Capsule, length=3, radius=2, shapeName=dyn_hair_bck_l_01_capsule, center=<2.287 0 0>, rotation=<rot 90 -2.5465 -2.5466>}]
 dyn_hair_bck_l_02 [{shapeType=Capsule, length=4, radius=4, shapeName=dyn_hair_bck_r_03_capsule, center=<2.287 0 0>, rotation=<rot 90 -2.927 -2.927>}]
 dyn_hair_bck_l_03 [{shapeType=Capsule, length=5, radius=4, shapeName=dyn_hair_bck_r_03_capsule, center=<2.287 0 0>, rotation=<rot 90 -5.5415 84.4584>}]
 dyn_hair_bck_l_04 [{shapeType=Capsule, length=5.5, radius=4.5, shapeName=dyn_hair_bck_r_03_capsule, center=<2.287 0 0>, rotation=<rot 90 1.2069 91.2069>}]
 dyn_hair_bck_l_05 [{shapeType=Capsule, length=4, radius=5.5, shapeName=dyn_hair_bck_r_03_capsule, center=<2.287 0 0>, rotation=<rot 90 -3.7012 86.2986>}]
 dyn_hair_bck_l_base [{shapeType=Capsule, length=0, radius=0.5, shapeName=dyn_hair_bck_l_base_capsule, rotation=<rot 90 0 0>}]
 dyn_hair_bck_r_01 [{shapeType=Capsule, length=3, radius=2, shapeName=dyn_hair_bck_l_01_capsule, center=<2.287 0 0>, rotation=<rot 90 -179.9999 179.9999>}]
 dyn_hair_bck_r_02 [{shapeType=Capsule, length=4.5, radius=3.5, shapeName=dyn_hair_bck_r_03_capsule, center=<2.287 0 0>, rotation=<rot 90 180 -180>}]
 dyn_hair_bck_r_03 [{shapeType=Capsule, length=2, radius=4.5, shapeName=dyn_hair_bck_r_03_capsule, center=<2.287 0 0>, rotation=<rot 90 1.1226 91.1226>}]
 dyn_hair_bck_r_04 [{shapeType=Capsule, length=4.5, radius=4.5, shapeName=dyn_hair_bck_r_03_capsule, center=<2.287 0 0>, rotation=<rot 90 1.2255 91.2255>}]
 dyn_hair_bck_r_05 [{shapeType=Capsule, length=6, radius=5.5, shapeName=dyn_hair_bck_r_03_capsule, center=<2.287 0 0>, rotation=<rot 90 1.2255 -88.7745>}]
 dyn_hair_bck_r_base [{shapeType=Capsule, length=0, radius=0.5, shapeName=dyn_hair_bck_r_base_capsule, rotation=<rot 90 0 0>}]
 dyn_hair_fr_l_01 [{
   shapeType=Capsule
   length=6.3555
   radius=2.8285
   shapeName=dyn_hair_fr_l_01_capsule
   center=<8.6552 -2.0881 -1.2955>
   rotation=<rot 80.5134 -141.7357 46.8663>
  }]
 dyn_hair_fr_l_02 [{
   shapeType=Capsule
   length=7
   radius=3.5
   shapeName=dyn_hair_fr_r_02_capsule
   center=<4.3237 0.4016 0.3982>
   rotation=<rot 75.7997 -86.3955 -147.04>
  }]
 dyn_hair_fr_l_03 [{
   shapeType=Box
   extentX=9.5
   extentY=11.5
   extentZ=4.5
   shapeName=dyn_hair_fr_r_03_box
   center=<3.5604 -0.396 0.1915>
   rotation=<rot -5.2134 0.0384 -0.2369>
  }]
 dyn_hair_fr_r_01 [{
   shapeType=Capsule
   length=7
   radius=2.5
   shapeName=dyn_hair_fr_l_01_capsule
   center=<9.7459 2.8339 -0.925>
   rotation=<rot 72.9423 -175.0516 3.6795>
  }]
 dyn_hair_fr_r_02 [{
   shapeType=Capsule
   length=6.5
   radius=2.5
   shapeName=dyn_hair_fr_r_02_capsule
   center=<4.6459 0.1369 0.5575>
   rotation=<rot 68.5341 125.8825 5.7483>
  }]
 dyn_hair_fr_r_03 [{
   shapeType=Box
   extentX=9.5
   extentY=11.5
   extentZ=6.5
   shapeName=dyn_hair_fr_r_03_box
   center=<3.2239 0.0883 -0.6354>
   rotation=<rot -14.9987 0 20>
  }]
 dyn_hair_side_l_01 [{shapeType=Capsule, length=3.5, radius=3.5, shapeName=dyn_hair_side_l_01_capsule, center=<3.5074 0 0>, rotation=<rot 90 -153.7666 -153.7666>}]
 dyn_hair_side_l_02 [{shapeType=Capsule, length=4, radius=3, shapeName=dyn_hair_side_l_02_capsule, center=<3.5074 0 0>, rotation=<rot 90 -166.4081 -166.4081>}]
 dyn_hair_side_l_03 [{
   shapeType=Capsule
   length=4.5
   radius=4
   shapeName=dyn_hair_side_l_03_capsule
   center=<3.5555 -0.4692 -0.5491>
   rotation=<rot 84.9999 179.9999 179.9999>
  }]
 dyn_hair_side_l_base [{shapeType=Capsule, length=0, radius=0.5, shapeName=dyn_hair_side_l_base_capsule, rotation=<rot 90 0 0>}]
 dyn_hair_side_r_01 [{shapeType=Capsule, length=3, radius=3, shapeName=dyn_hair_side_r_01_capsule, center=<3.2391 0 0>, rotation=<rot 90 0 0>}]
 dyn_hair_side_r_02 [{shapeType=Capsule, length=2.5, radius=3.5, shapeName=dyn_hair_side_r_02_capsule, center=<3.2391 0 0>, rotation=<rot 90 0 0>}]
 dyn_hair_side_r_03 [{
   shapeType=Capsule
   length=5.5
   radius=3.5
   shapeName=dyn_hair_side_r_03_capsule
   center=<5.0456 0 -2.2151>
   rotation=<rot 74.9998 -0.0001 -0.0001>
  }]
 dyn_hair_side_r_base [{shapeType=Capsule, length=0, radius=0.5, shapeName=dyn_hair_side_r_base_capsule, rotation=<rot 90 0 0>}]
 hair_fr_l_base [{shapeType=Capsule, length=0.505, radius=0.505, shapeName=hair_fr_l_base_capsule, rotation=<rot 0 -90 0>}]
 hair_fr_r_base [{shapeType=Capsule, length=0.505, radius=0.505, shapeName=hair_fr_r_base_capsule, rotation=<rot 0 -90 0>}]
 head [{shapeType=Capsule, length=7, radius=8, shapeName=head_capsule, center=<2.827 3.6062 0>, rotation=<rot 0 -78.8774 -90>}]
 neck_01 [{shapeType=Capsule, length=0.5, radius=6, shapeName=neck_01_capsule, rotation=<rot 0 -0.0003 0>}]
 neck_02 [{shapeType=Capsule, length=2.5, radius=5.5, shapeName=neck_02_capsule, rotation=<rot 90 164.8051 164.8053>}]
 spine_04 [{shapeType=Capsule, length=17.005, radius=12.5, shapeName=spine_04_capsule, center=<3.6309 0.2268 0>, rotation=<rot 0 -90 0>}]
 spine_05 [
  {shapeType=Capsule, length=19.5, radius=9, shapeName=spine_05_capsule, center=<2.8113 -1.0118 0>, rotation=<rot 0 -0.0002 0>}
  {shapeType=Capsule, length=6, radius=35, shapeName=spine_05_capsule, center=<25.8919 -51.2674 0>}
 ]
 upperarm_l [{shapeType=Capsule, length=23.5, radius=5.5, shapeName=upperarm_l_capsule, center=<8.3369 0 0>, rotation=<rot 90 0 0>}]
 upperarm_r [{shapeType=Capsule, length=25, radius=7, shapeName=upperarm_r_capsule, center=<-13 0 0>, rotation=<rot -90 0 0>}]
}
constraints [
 {bone1Name=dyn_hair_bck_l_base, bone2Name=head, swing1LimitDegrees=10, swing2LimitDegrees=45, twistLimitDegrees=2}
 {bone1Name=dyn_hair_bck_l_01, bone2Name=dyn_hair_bck_l_base, swing1LimitDegrees=10, swing2LimitDegrees=45, twistLimitDegrees=2}
 {bone1Name=dyn_hair_bck_l_02, bone2Name=dyn_hair_bck_l_01, swing1LimitDegrees=20, swing2LimitDegrees=45, twistLimitDegrees=20}
 {bone1Name=dyn_hair_bck_l_03, bone2Name=dyn_hair_bck_l_02, swing1LimitDegrees=10, swing2LimitDegrees=40, twistLimitDegrees=20}
 {bone1Name=dyn_hair_bck_l_04, bone2Name=dyn_hair_bck_l_03, swing1LimitDegrees=20, swing2LimitDegrees=45, twistLimitDegrees=30}
 {bone1Name=dyn_hair_bck_l_05, bone2Name=dyn_hair_bck_l_04, swing1LimitDegrees=20, swing2LimitDegrees=45, twistLimitDegrees=20}
 {bone1Name=dyn_hair_bck_r_base, bone2Name=head, swing1LimitDegrees=10, swing2LimitDegrees=45, twistLimitDegrees=2}
 {bone1Name=dyn_hair_bck_r_01, bone2Name=dyn_hair_bck_r_base, swing1LimitDegrees=10, swing2LimitDegrees=45, twistLimitDegrees=2}
 {bone1Name=dyn_hair_bck_r_02, bone2Name=dyn_hair_bck_r_01, swing1LimitDegrees=20, swing2LimitDegrees=45, twistLimitDegrees=20}
 {bone1Name=dyn_hair_bck_r_03, bone2Name=dyn_hair_bck_r_02, swing1LimitDegrees=10, swing2LimitDegrees=40, twistLimitDegrees=2}
 {bone1Name=dyn_hair_bck_r_04, bone2Name=dyn_hair_bck_r_03, swing1LimitDegrees=10, swing2LimitDegrees=45, twistLimitDegrees=2}
 {bone1Name=dyn_hair_bck_r_05, bone2Name=dyn_hair_bck_r_04, swing1LimitDegrees=10, swing2LimitDegrees=45, twistLimitDegrees=2}
 {bone1Name=dyn_hair_side_l_base, bone2Name=head, swing1LimitDegrees=10, swing2LimitDegrees=45, twistLimitDegrees=45}
 {bone1Name=dyn_hair_side_l_01, bone2Name=dyn_hair_side_l_base, swing1LimitDegrees=20, swing2LimitDegrees=20, twistLimitDegrees=10}
 {bone1Name=dyn_hair_side_l_02, bone2Name=dyn_hair_side_l_01, swing1LimitDegrees=10, swing2LimitDegrees=20, twistLimitDegrees=10}
 {bone1Name=dyn_hair_side_l_03, bone2Name=dyn_hair_side_l_02, swing1LimitDegrees=10, swing2LimitDegrees=20, twistLimitDegrees=10}
 {bone1Name=dyn_hair_side_r_base, bone2Name=head, swing1LimitDegrees=10, swing2LimitDegrees=45, twistLimitDegrees=45}
 {bone1Name=dyn_hair_side_r_01, bone2Name=dyn_hair_side_r_base, swing1LimitDegrees=10, swing2LimitDegrees=40, twistLimitDegrees=2}
 {bone1Name=dyn_hair_side_r_02, bone2Name=dyn_hair_side_r_01, swing1LimitDegrees=15, swing2LimitDegrees=30, twistLimitDegrees=2}
 {bone1Name=dyn_hair_side_r_03, bone2Name=dyn_hair_side_r_02, swing1LimitDegrees=30, swing2LimitDegrees=30, twistLimitDegrees=2}
 {
  bone1Name=dyn_hair_bck_l_04
  bone2Name=dyn_hair_bck_r_04
  swing1LimitDegrees=45
  swing1Motion=Free
  swing2LimitDegrees=45
  swing2Motion=Free
  twistLimitDegrees=45
  twistMotion=Free
 }
 {
  bone1Name=dyn_hair_side_r_03
  bone2Name=dyn_hair_bck_r_03
  swing1LimitDegrees=45
  swing1Motion=Free
  swing2LimitDegrees=45
  swing2Motion=Free
  twistLimitDegrees=45
  twistMotion=Free
 }
 {
  bone1Name=dyn_hair_side_l_03
  bone2Name=dyn_hair_bck_l_03
  swing1LimitDegrees=45
  swing1Motion=Free
  swing2LimitDegrees=45
  swing2Motion=Free
  twistLimitDegrees=45
  twistMotion=Free
 }
 {
  bone1Name=dyn_hair_side_l_02
  bone2Name=dyn_hair_bck_l_02
  swing1LimitDegrees=45
  swing1Motion=Free
  swing2LimitDegrees=45
  swing2Motion=Free
  twistLimitDegrees=45
  twistMotion=Free
 }
 {
  bone1Name=dyn_hair_bck_l_05
  bone2Name=dyn_hair_bck_r_05
  swing1LimitDegrees=45
  swing1Motion=Free
  swing2LimitDegrees=45
  swing2Motion=Free
  twistLimitDegrees=45
  twistMotion=Free
 }
 {
  bone1Name=dyn_hair_bck_l_04
  bone2Name=spine_05
  swing1LimitDegrees=45
  swing1Motion=Free
  swing2LimitDegrees=45
  swing2Motion=Free
  twistLimitDegrees=45
  twistMotion=Free
 }
 {
  bone1Name=dyn_hair_bck_r_04
  bone2Name=spine_05
  swing1LimitDegrees=45
  swing1Motion=Free
  swing2LimitDegrees=45
  swing2Motion=Free
  twistLimitDegrees=45
  twistMotion=Free
 }
 {
  bone1Name=hair_fr_l_base
  bone2Name=neck_01
  swing1LimitDegrees=45
  swing1Motion=Free
  swing2LimitDegrees=45
  swing2Motion=Free
  twistLimitDegrees=45
  twistMotion=Free
 }
 {bone1Name=dyn_hair_fr_l_01, bone2Name=hair_fr_l_base, swing1LimitDegrees=30, swing2LimitDegrees=40, twistLimitDegrees=5}
 {bone1Name=dyn_hair_fr_l_02, bone2Name=dyn_hair_fr_l_01, swing1LimitDegrees=20, swing2LimitDegrees=40, twistLimitDegrees=45}
 {bone1Name=dyn_hair_fr_l_03, bone2Name=dyn_hair_fr_l_02, swing1LimitDegrees=20, swing2LimitDegrees=45, twistLimitDegrees=30}
 {
  bone1Name=hair_fr_r_base
  bone2Name=neck_01
  swing1LimitDegrees=45
  swing1Motion=Free
  swing2LimitDegrees=45
  swing2Motion=Free
  twistLimitDegrees=45
  twistMotion=Free
 }
 {bone1Name=dyn_hair_fr_r_01, bone2Name=hair_fr_r_base, swing1LimitDegrees=20, swing2LimitDegrees=40, twistLimitDegrees=20}
 {bone1Name=dyn_hair_fr_r_02, bone2Name=dyn_hair_fr_r_01, swing1LimitDegrees=10, swing2LimitDegrees=40, twistLimitDegrees=45}
 {bone1Name=dyn_hair_fr_r_03, bone2Name=dyn_hair_fr_r_02, swing1LimitDegrees=40, swing2LimitDegrees=40, twistLimitDegrees=30}
 {
  bone1Name=hair_fr_l_base
  bone2Name=head
  swing1LimitDegrees=45
  swing1Motion=Free
  swing2LimitDegrees=45
  swing2Motion=Free
  twistLimitDegrees=45
  twistMotion=Free
 }
]
