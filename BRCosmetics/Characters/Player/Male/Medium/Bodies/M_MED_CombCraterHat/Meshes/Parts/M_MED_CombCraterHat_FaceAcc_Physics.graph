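# Physics asset for the face accessory part of the CombCraterHat male medium character, simulating its hair and hat secondary motion.
# Back and side hair chains hang from the head, front chains from neck_01, and three hat pieces from spine_05 via small base bodies.
# Head, neck, spine, clavicle and upper-arm shapes supply body volumes.
# Narrow angular limits shape most chains, while free-angle cross constraints tie side to back hair and left to right.
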PhysicsAsset /BRCosmetics/Characters/Player/Male/Medium/Bodies/M_MED_CombCraterHat/Meshes/Parts/M_MED_CombCraterHat_FaceAcc_Physics
bodies {
 base_hat_bk [{shapeType=Sphere, radius=1, shapeName=base_hat_bk_sphere}]
 base_hat_lb [{shapeType=Sphere, radius=1, shapeName=base_hat_lb_sphere}]
 base_hat_rb [{shapeType=Sphere, radius=1, shapeName=base_hat_rb_sphere}]
 clavicle_l [{
   shapeType=Capsule
   length=17.5
   radius=6
   shapeName=clavicle_l_capsule
   center=<10.5557 -1.5456 3.7067>
   rotation=<rot 71.153 1.2999 -0.9604>
  }]
 clavicle_r [{
   shapeType=Capsule
   length=17.5
   radius=7
   shapeName=clavicle_r_capsule
   center=<-9.8407 1.8876 -2.7002>
   rotation=<rot -75.538 -175.0967 178.3151>
  }]
 dyn_hair_bck_l_01 [{shapeType=Capsule, length=3, radius=2, shapeName=dyn_hair_bck_l_01_capsule, center=<3.5 0 0>, rotation=<rot 90 -179.9999 179.9999>}]
 dyn_hair_bck_l_02 [{shapeType=Capsule, length=4, radius=3, shapeName=dyn_hair_bck_r_03_capsule, center=<4.5 0 0>, rotation=<rot 90 180 -180>}]
 dyn_hair_bck_l_03 [{shapeType=Capsule, length=1.5, radius=3.5, shapeName=dyn_hair_bck_r_03_capsule, center=<3.5 0 0>, rotation=<rot 90 0 0>}]
 dyn_hair_bck_l_04 [{shapeType=Capsule, length=6, radius=3.5, shapeName=dyn_hair_bck_r_03_capsule, center=<2.2944 -0.0283 0>, rotation=<rot 90 180 -89.2939>}]
 dyn_hair_bck_l_05 [{shapeType=Capsule, length=1, radius=4, shapeName=dyn_hair_bck_r_03_capsule, center=<2.0629 0 -0.7927>, rotation=<rot 90 0 0>}]
 dyn_hair_bck_l_base [{shapeType=Capsule, length=0, radius=0.5, shapeName=dyn_hair_bck_l_base_capsule, rotation=<rot 90 0 0>}]
 dyn_hair_bck_r_01 [{shapeType=Capsule, length=3, radius=2, shapeName=dyn_hair_bck_l_01_capsule, center=<3.5 0 0>, rotation=<rot 90 -179.9999 179.9999>}]
 dyn_hair_bck_r_02 [{shapeType=Capsule, length=4, radius=3, shapeName=dyn_hair_bck_r_03_capsule, center=<4.5 0 0>, rotation=<rot 90 180 -180>}]
 dyn_hair_bck_r_03 [{shapeType=Capsule, length=1.5, radius=3.5, shapeName=dyn_hair_bck_r_03_capsule, center=<3.5 0 0>, rotation=<rot 90 0 0>}]
 dyn_hair_bck_r_04 [{shapeType=Capsule, length=6, radius=3.5, shapeName=dyn_hair_bck_r_03_capsule, center=<2.2944 -0.0283 0>, rotation=<rot 90 180 -89.2939>}]
 dyn_hair_bck_r_05 [{shapeType=Capsule, length=1, radius=4, shapeName=dyn_hair_bck_r_03_capsule, center=<2.0629 0 -0.7927>, rotation=<rot 90 0 0>}]
 dyn_hair_bck_r_base [{shapeType=Capsule, length=0, radius=0.5, shapeName=dyn_hair_bck_r_base_capsule, rotation=<rot 90 0 0>}]
 dyn_hair_fr_l_01 [{
   shapeType=Capsule
   length=4.8555
   radius=3.3285
   shapeName=dyn_hair_fr_l_01_capsule
   center=<7.9968 -0.1676 -0.9986>
   rotation=<rot 0.8616 -112.0885 99.9545>
  }]
 dyn_hair_fr_l_02 [{
   shapeType=Capsule
   length=8.5
   radius=2
   shapeName=dyn_hair_fr_r_02_capsule
   center=<4.4962 0.1044 -0.5477>
   rotation=<rot 74.5864 -160.8666 107.9709>
  }]
 dyn_hair_fr_l_03 [{shapeType=Box, extentX=9.5, extentY=11.5, extentZ=1.5, shapeName=dyn_hair_fr_r_03_box, center=<3.3162 0 -0.0018>}]
 dyn_hair_fr_r_01 [{
   shapeType=Capsule
   length=6
   radius=3.5
   shapeName=dyn_hair_fr_l_01_capsule
   center=<4.6987 1.3208 0.0578>
   rotation=<rot 78.6591 -90.2252 64.0505>
  }]
 dyn_hair_fr_r_02 [{
   shapeType=Capsule
   length=7.5
   radius=3
   shapeName=dyn_hair_fr_r_02_capsule
   center=<1.4428 0.0127 0.022>
   rotation=<rot 87.4946 177.4508 86.9454>
  }]
 dyn_hair_fr_r_03 [{shapeType=Box, extentX=9.5, extentY=11.5, extentZ=4.5, shapeName=dyn_hair_fr_r_03_box, center=<3.3162 0 -0.0018>}]
 dyn_hair_side_l_01 [{shapeType=Capsule, length=3.5, radius=2, shapeName=dyn_hair_side_l_01_capsule, center=<3.5 0 0>, rotation=<rot 90 0 0>}]
 dyn_hair_side_l_02 [{shapeType=Capsule, length=4, radius=3, shapeName=dyn_hair_side_l_02_capsule, center=<4.5 0 0>, rotation=<rot 90 0 0>}]
 dyn_hair_side_l_03 [{shapeType=Capsule, length=1.5, radius=3.5, shapeName=dyn_hair_side_l_03_capsule, center=<3.5 0 0>, rotation=<rot 90 0 0>}]
 dyn_hair_side_l_base [{shapeType=Capsule, length=0, radius=0.5, shapeName=dyn_hair_side_l_base_capsule, rotation=<rot 90 0 0>}]
 dyn_hair_side_r_01 [{shapeType=Capsule, length=3.5, radius=2, shapeName=dyn_hair_side_r_01_capsule, center=<3.5 0 0>, rotation=<rot 90 0 0>}]
 dyn_hair_side_r_02 [{shapeType=Capsule, length=4, radius=3, shapeName=dyn_hair_side_r_02_capsule, center=<4.5 0.5076 0.6692>, rotation=<rot 51.245 89.9999 90>}]
 dyn_hair_side_r_03 [{shapeType=Capsule, length=1.5, radius=3.5, shapeName=dyn_hair_side_r_03_capsule, center=<3.5 0 0.4948>, rotation=<rot 64.5174 0 0>}]
 dyn_hair_side_r_base [{shapeType=Capsule, length=0, radius=0.5, shapeName=dyn_hair_side_r_base_capsule, rotation=<rot 90 0 0>}]
 dyn_hat_bk [{shapeType=Capsule, length=14.5, radius=3.5, shapeName=dyn_hat_bk_capsule, center=<5.8136 0.0595 -2>, rotation=<rot 0 0 90>}]
 dyn_hat_lb [{shapeType=Capsule, length=14.5, radius=3.5, shapeName=dyn_hat_lb_capsule, center=<5.8136 0.1844 -2>, rotation=<rot 0 0 81.0471>}]
 dyn_hat_rb [{shapeType=Capsule, length=15, radius=3, shapeName=dyn_hat_rb_capsule, center=<5.8136 0 -2>, rotation=<rot 0 0 93.6239>}]
 hair_fr_l_base [{shapeType=Capsule, length=0.505, radius=0.505, shapeName=hair_fr_l_base_capsule, rotation=<rot 0 -90 0>}]
 hair_fr_r_base [{shapeType=Capsule, length=0.505, radius=0.505, shapeName=hair_fr_r_base_capsule, rotation=<rot 0 -90 0>}]
 head [{shapeType=Capsule, length=10, radius=8, shapeName=head_capsule, center=<4.2181 2.6374 0>, rotation=<rot 0 -78.8774 -90>}]
 neck_01 [{shapeType=Capsule, length=10.5, radius=5.5, shapeName=neck_01_capsule, rotation=<rot 90 0 0>}]
 neck_02 [{shapeType=Capsule, length=2.5, radius=5.5, shapeName=neck_02_capsule, rotation=<rot 90 164.8051 164.8053>}]
 spine_03 [{shapeType=Capsule, length=17.005, radius=12.5, shapeName=spine_04_capsule, rotation=<rot 0 -90 0>}]
 spine_04 [{shapeType=Capsule, length=18.505, radius=12.5, shapeName=spine_04_capsule, rotation=<rot 0 -90 0>}]
 spine_05 [
  {shapeType=Capsule, length=17.5, radius=11, shapeName=spine_05_capsule, center=<0.1467 -0.6794 0>}
  {shapeType=Capsule, length=18.5, radius=11.5, shapeName=spine_05_capsule, center=<-4.0407 1.4901 0>}
 ]
 upperarm_l [{shapeType=Capsule, length=25, radius=7, shapeName=upperarm_l_capsule, center=<15.4108 0 0>, rotation=<rot 90 0 0>}]
 upperarm_r [{shapeType=Capsule, length=21.5, radius=7.5, shapeName=upperarm_r_capsule, center=<-13.6205 0 0>, rotation=<rot -90 0 0>}]
}
constraints [
 {bone1Name=dyn_hair_bck_l_base, bone2Name=head, swing1LimitDegrees=10, swing2LimitDegrees=45, twistLimitDegrees=2}
 {bone1Name=dyn_hair_bck_l_01, bone2Name=dyn_hair_bck_l_base, swing1LimitDegrees=10, swing2LimitDegrees=45, twistLimitDegrees=2}
 {bone1Name=dyn_hair_bck_l_02, bone2Name=dyn_hair_bck_l_01, swing1LimitDegrees=10, swing2LimitDegrees=40, twistLimitDegrees=2}
 {bone1Name=dyn_hair_bck_l_03, bone2Name=dyn_hair_bck_l_02, swing1LimitDegrees=10, swing2LimitDegrees=40, twistLimitDegrees=3}
 {bone1Name=dyn_hair_bck_l_04, bone2Name=dyn_hair_bck_l_03, swing1LimitDegrees=10, swing2LimitDegrees=40, twistLimitDegrees=2}
 {bone1Name=dyn_hair_bck_l_05, bone2Name=dyn_hair_bck_l_04, swing1LimitDegrees=10, swing2LimitDegrees=45, twistLimitDegrees=2}
 {
  bone1Name=dyn_hair_bck_r_base
  bone2Name=head
  swing1LimitDegrees=10
  swing1Motion=Free
  swing2LimitDegrees=45
  swing2Motion=Free
  twistLimitDegrees=2
  twistMotion=Free
 }
 {bone1Name=dyn_hair_bck_r_01, bone2Name=dyn_hair_bck_r_base, swing1LimitDegrees=10, swing2LimitDegrees=45, twistLimitDegrees=2}
 {bone1Name=dyn_hair_bck_r_02, bone2Name=dyn_hair_bck_r_01, swing1LimitDegrees=10, swing2LimitDegrees=40, twistLimitDegrees=2}
 {bone1Name=dyn_hair_bck_r_03, bone2Name=dyn_hair_bck_r_02, swing1LimitDegrees=10, swing2LimitDegrees=40, twistLimitDegrees=2}
 {bone1Name=dyn_hair_bck_r_04, bone2Name=dyn_hair_bck_r_03, swing1LimitDegrees=10, swing2LimitDegrees=40, twistLimitDegrees=2}
 {bone1Name=dyn_hair_bck_r_05, bone2Name=dyn_hair_bck_r_04, swing1LimitDegrees=10, swing2LimitDegrees=45, twistLimitDegrees=2}
 {
  bone1Name=dyn_hair_side_l_base
  bone2Name=head
  swing1LimitDegrees=10
  swing1Motion=Free
  swing2LimitDegrees=45
  swing2Motion=Free
  twistLimitDegrees=45
  twistMotion=Free
 }
 {bone1Name=dyn_hair_side_l_01, bone2Name=dyn_hair_side_l_base, swing1LimitDegrees=20, swing2LimitDegrees=20, twistLimitDegrees=2}
 {bone1Name=dyn_hair_side_l_02, bone2Name=dyn_hair_side_l_01, swing1LimitDegrees=10, swing2LimitDegrees=10, twistLimitDegrees=2}
 {bone1Name=dyn_hair_side_l_03, bone2Name=dyn_hair_side_l_02, swing1LimitDegrees=10, swing2LimitDegrees=10, twistLimitDegrees=2}
 {bone1Name=dyn_hair_side_r_base, bone2Name=head, swing1LimitDegrees=10, swing2LimitDegrees=45, twistLimitDegrees=45}
 {bone1Name=dyn_hair_side_r_01, bone2Name=dyn_hair_side_r_base, swing1LimitDegrees=20, swing2LimitDegrees=40, twistLimitDegrees=2}
 {bone1Name=dyn_hair_side_r_02, bone2Name=dyn_hair_side_r_01, swing1LimitDegrees=15, swing2LimitDegrees=30, twistLimitDegrees=2}
 {bone1Name=dyn_hair_side_r_03, bone2Name=dyn_hair_side_r_02, swing1LimitDegrees=30, swing2LimitDegrees=30, twistLimitDegrees=2}
 {
  bone1Name=dyn_hair_side_r_03
  bone2Name=dyn_hair_bck_r_03
  swing1LimitDegrees=45
  swing1Motion=Free
  swing2LimitDegrees=45
  swing2Motion=Free
  twistLimitDegrees=45
  twistMotion=Free
 }
 {
  bone1Name=dyn_hair_side_l_03
  bone2Name=dyn_hair_bck_l_03
  swing1LimitDegrees=45
  swing1Motion=Free
  swing2LimitDegrees=45
  swing2Motion=Free
  twistLimitDegrees=45
  twistMotion=Free
 }
 {
  bone1Name=dyn_hair_side_l_02
  bone2Name=dyn_hair_bck_l_02
  swing1LimitDegrees=45
  swing1Motion=Free
  swing2LimitDegrees=45
  swing2Motion=Free
  twistLimitDegrees=45
  twistMotion=Free
 }
 {
  bone1Name=dyn_hair_bck_l_02
  bone2Name=dyn_hair_bck_r_02
  swing1LimitDegrees=45
  swing1Motion=Free
  swing2LimitDegrees=45
  swing2Motion=Free
  twistLimitDegrees=45
  twistMotion=Free
 }
 {
  bone1Name=dyn_hair_bck_l_05
  bone2Name=dyn_hair_bck_r_05
  swing1LimitDegrees=45
  swing1Motion=Free
  swing2LimitDegrees=45
  swing2Motion=Free
  twistLimitDegrees=45
  twistMotion=Free
 }
 {
  bone1Name=hair_fr_l_base
  bone2Name=neck_01
  swing1LimitDegrees=45
  swing1Motion=Free
  swing2LimitDegrees=45
  swing2Motion=Free
  twistLimitDegrees=45
  twistMotion=Free
 }
 {bone1Name=dyn_hair_fr_l_01, bone2Name=hair_fr_l_base, swing1LimitDegrees=10, swing2LimitDegrees=40, twistLimitDegrees=15}
 {bone1Name=dyn_hair_fr_l_02, bone2Name=dyn_hair_fr_l_01, swing1LimitDegrees=20, swing2LimitDegrees=40, twistLimitDegrees=20}
 {bone1Name=dyn_hair_fr_l_03, bone2Name=dyn_hair_fr_l_02, swing1LimitDegrees=10, swing2LimitDegrees=40, twistLimitDegrees=15}
 {
  bone1Name=hair_fr_r_base
  bone2Name=neck_01
  swing1LimitDegrees=45
  swing1Motion=Free
  swing2LimitDegrees=45
  swing2Motion=Free
  twistLimitDegrees=45
  twistMotion=Free
 }
 {bone1Name=dyn_hair_fr_r_01, bone2Name=hair_fr_r_base, swing1LimitDegrees=30, swing2LimitDegrees=40, twistLimitDegrees=15}
 {bone1Name=dyn_hair_fr_r_02, bone2Name=dyn_hair_fr_r_01, swing1LimitDegrees=10, swing2LimitDegrees=40, twistLimitDegrees=20}
 {bone1Name=dyn_hair_fr_r_03, bone2Name=dyn_hair_fr_r_02, swing1LimitDegrees=20, swing2LimitDegrees=40, twistLimitDegrees=8}
 {bone1Name=dyn_hat_rb, bone2Name=base_hat_rb, swing1LimitDegrees=1, swing2LimitDegrees=30, twistLimitDegrees=1}
 {bone1Name=dyn_hat_bk, bone2Name=base_hat_bk, swing1LimitDegrees=1, swing2LimitDegrees=30, twistLimitDegrees=10}
 {
  bone1Name=base_hat_bk
  bone2Name=spine_05
  swing1LimitDegrees=45
  swing1Motion=Free
  swing2LimitDegrees=45
  swing2Motion=Free
  twistLimitDegrees=45
  twistMotion=Free
 }
 {bone1Name=dyn_hat_lb, bone2Name=base_hat_lb, swing1LimitDegrees=1, swing2LimitDegrees=20, twistLimitDegrees=1}
 {
  bone1Name=base_hat_lb
  bone2Name=spine_05
  swing1LimitDegrees=45
  swing1Motion=Free
  swing2LimitDegrees=45
  swing2Motion=Free
  twistLimitDegrees=45
  twistMotion=Free
 }
 {
  bone1Name=base_hat_rb
  bone2Name=spine_05
  swing1LimitDegrees=45
  swing1Motion=Free
  swing2LimitDegrees=45
  swing2Motion=Free
  twistLimitDegrees=45
  twistMotion=Free
 }
]
