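# Face-accessory AnimBlueprint for the male CombCrater Trade skin that drives its hair bones with physics.
# It merges attached-parent and body clavicle poses, constrains and offsets hair roots, then caches the result before simulation.
# Jacket, front-end, backbling and skydiving flags select among RigidBody sims using state-driven gravity.
# Afterwards, front strands are constrained to spine_05 at front-end-dependent strengths, and parachuting adds root offsets.
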AnimBlueprint /BRCosmetics/Characters/Player/Male/Medium/Bodies/M_MED_CombCrater_Trade/Meshes/Parts/M_MED_CombCrater_Trade_FaceAcc_AnimBP
target_skeleton=&/M_MED_CombCrater_Trade_FaceAcc_Skeleton
graphs [
 AnimGraph {nodes [
   "Transform (Modify) Bone - Bone: hair_fr_l_base"(Alpha=1, Translation="0,0,0", Rotation="-10.258,1.304,-18.170", Scale="1,1,1") {
    rotationMode=BMM_Additive
    translationMode=BMM_Additive
    alphaBoolBlend {blendOption=Linear}
    boneToModify {boneName=hair_fr_l_base}
    rotation=<rot -10.258 1.304 -18.17>
   }
   n1 = "Transform (Modify) Bone - Bone: hair_fr_l_base"(Alpha=1, Translation="0.850,0.000,0.000", Rotation="10.703,-63.870,-3.896", Scale="1,1,1") {
    rotationMode=BMM_Additive
    translationMode=BMM_Additive
    alphaBoolBlend {blendOption=Linear}
    boneToModify {boneName=hair_fr_l_base}
    rotation=<rot 10.703 -63.87 -3.896>
    translation {x=0.85}
   }
   "Transform (Modify) Bone - Bone: hair_fr_r_base"(ComponentPose=@n1, Alpha=1, Translation="0.000,0.000,3.043", Rotation="0,0,-50", Scale="1,1,1") {
    rotationMode=BMM_Additive
    translationMode=BMM_Additive
    alphaBoolBlend {blendOption=Linear}
    boneToModify {boneName=hair_fr_r_base}
    rotation {roll=-50}
    translation {z=3.043}
   }
   n2 = "Transform (Modify) Bone - Bone: dyn_hair_bck_l_base"(Alpha=1, Translation="0.000,-1.483,0.000", Rotation="0,0,25", Scale="1,1,1") {
    rotationMode=BMM_Additive
    translationMode=BMM_Additive
    alphaBoolBlend {blendOption=Linear}
    alphaScaleBiasClamp {bInterpResult=true}
    boneToModify {boneName=dyn_hair_bck_l_base}
    rotation {roll=25}
    translation {y=-1.483}
   }
   n3 = "Transform (Modify) Bone - Bone: dyn_hair_bck_r_base"(ComponentPose=@n2, Alpha=1, Translation="0.000,-1.456,0.863", Rotation="0,0,25", Scale="1,1,1") {
    rotationMode=BMM_Additive
    translationMode=BMM_Additive
    alphaBoolBlend {blendOption=Linear}
    alphaScaleBiasClamp {bInterpResult=true}
    boneToModify {boneName=dyn_hair_bck_r_base}
    rotation {roll=25}
    translation {y=-1.456, z=0.863}
   }
   n4 = "Transform (Modify) Bone - Bone: dyn_hair_side_l_base"(ComponentPose=@n3, Alpha=1, Translation="0,0,0", Rotation="0,0,0", Scale="1,1,1") {
    rotationMode=BMM_Additive
    translationMode=BMM_Additive
    alphaBoolBlend {blendOption=Linear}
    alphaScaleBiasClamp {bInterpResult=true}
    boneToModify {boneName=dyn_hair_side_l_base}
   }
   "Transform (Modify) Bone - Bone: dyn_hair_side_r_base"(ComponentPose=@n4, Alpha=1, Translation="0,0,0", Rotation="-0.000,0.000,18.302", Scale="1,1,1") {
    rotationMode=BMM_Additive
    translationMode=BMM_Additive
    alphaBoolBlend {blendOption=Linear}
    alphaScaleBiasClamp {bInterpResult=true}
    boneToModify {boneName=dyn_hair_side_r_base}
    rotation {roll=18.302}
   }
   "Transform (Modify) Bone - Bone: hair_fr_r_base"(Alpha=1, Translation="0.000,0.000,3.043", Rotation="16.786,14.929,-52.343", Scale="1,1,1") {
    rotationMode=BMM_Additive
    translationMode=BMM_Additive
    alphaBoolBlend {blendOption=Linear}
    boneToModify {boneName=hair_fr_r_base}
    rotation=<rot 16.786 14.929 -52.343>
    translation {z=3.043}
   }
   "Transform (Modify) Bone - Bone: hair_fr_l_base"(Alpha=1, Translation="0.850,0.000,0.000", Rotation="10.703,-63.870,-3.896", Scale="1,1,1") {
    rotationMode=BMM_Additive
    translationMode=BMM_Additive
    alphaBoolBlend {blendOption=Linear}
    boneToModify {boneName=hair_fr_l_base}
    rotation=<rot 10.703 -63.87 -3.896>
    translation {x=0.85}
   }
   RigidBody!(ComponentPose, Alpha, OverrideWorldGravity:Vector=Get GravityOverride(), ExternalForce) -> (Pose) {
    bClampLinearTranslationLimitToRefPose=true
    bOverrideWorldGravity=true
    overridePhysicsAsset=/BRCosmetics/Characters/Player/Male/Medium/Bodies/M_MED_CombCrater/Meshes/Parts/M_MED_CombCrater_Front_FaceAcc_Physics
    simulationSpace=BaseBoneSpace
    alphaBoolBlend {blendOption=Linear}
    alphaScaleBiasClamp {bInterpResult=true}
    baseBoneRef {boneName=spine_05}
    componentAppliedLinearAccClamp=<3000 3000 3000>
    overrideWorldGravity {z=-980}
    simSpaceSettings {maxAngularVelocity=3, maxLinearVelocity=0.3, worldAlpha=0.35}
   }
   n5 = RigidBody(ComponentPose=Local To Component(LocalPose=Use cached pose 'Pre sim'()), OverrideWorldGravity=Get GravityOverride()) {
    bClampLinearTranslationLimitToRefPose=true
    bOverrideWorldGravity=true
    overridePhysicsAsset=/BRCosmetics/Characters/Player/Male/Medium/Bodies/M_MED_CombCraterHat/Meshes/Parts/M_MED_CombCraterHat_FaceAcc_Physics
    simulationSpace=BaseBoneSpace
    alphaBoolBlend {blendOption=Linear}
    baseBoneRef {boneName=neck_02}
    overrideWorldGravity {z=-980}
    simSpaceSettings {maxAngularAcceleration=65, maxAngularVelocity=15, maxLinearAcceleration=1000, maxLinearVelocity=3, velocityScaleZ=0.1, worldAlpha=0.3}
   }
   n6 = Component To Local(ComponentPose=@n5)
   n7 = RigidBody(ComponentPose=Local To Component(LocalPose=Use cached pose 'Pre sim'()), OverrideWorldGravity=Get GravityOverride()) {
    bClampLinearTranslationLimitToRefPose=true
    bOverrideWorldGravity=true
    overridePhysicsAsset=&M_MED_CombCraterHat_FaceAcc_Physics
    simulationSpace=BaseBoneSpace
    alphaBoolBlend {blendOption=Linear}
    baseBoneRef {boneName=neck_02}
    overrideWorldGravity {z=-980}
    simSpaceSettings {maxAngularAcceleration=65, maxAngularVelocity=15, maxLinearAcceleration=1000, maxLinearVelocity=0.3, worldAlpha=0.45}
   }
   n8 = Component To Local(ComponentPose=@n7)
   n9 = Get bIsSkyDiving()
   n10 = RigidBody(ComponentPose=Local To Component(LocalPose=Use cached pose 'Pre sim'()), OverrideWorldGravity=Get GravityOverride()) {
    bClampLinearTranslationLimitToRefPose=true
    bOverrideWorldGravity=true
    overridePhysicsAsset=/BRCosmetics/Characters/Player/Male/Medium/Bodies/M_MED_CombCraterHat/Meshes/Parts/M_MED_CombCraterHat_BackBling_01_FaceAcc_Physics
    simulationSpace=BaseBoneSpace
    alphaBoolBlend {blendOption=Linear}
    baseBoneRef {boneName=neck_02}
    componentAppliedLinearAccClamp=<0 0 0>
    overrideWorldGravity {z=-980}
    simSpaceSettings {maxAngularAcceleration=10, maxAngularVelocity=3, maxLinearAcceleration=10, maxLinearVelocity=3, velocityScaleZ=0.1, worldAlpha=0.8}
   }
   n11 = RigidBody(ComponentPose=Local To Component(LocalPose=Use cached pose 'Pre sim'()), OverrideWorldGravity=Get GravityOverride()) {
    bClampLinearTranslationLimitToRefPose=true
    bOverrideWorldGravity=true
    overridePhysicsAsset=/BRCosmetics/Characters/Player/Male/Medium/Bodies/M_MED_CombCraterHat/Meshes/Parts/M_MED_CombCraterHat_Backbling_FaceAcc_Physics
    simulationSpace=BaseBoneSpace
    alphaBoolBlend {blendOption=Linear}
    baseBoneRef {boneName=neck_02}
    overrideWorldGravity {z=-980}
    simSpaceSettings {maxAngularAcceleration=65, maxAngularVelocity=15, maxLinearAcceleration=1000, maxLinearVelocity=3, velocityScaleZ=0.1, worldAlpha=0.3}
   }
   n12 = Blend Poses by bool(
    BlendPose_0=Component To Local(ComponentPose=@n11)
    BlendPose_1=Blend Poses by bool(
     BlendPose_0=Component To Local(ComponentPose=@n10)
     BlendPose_1=Blend Poses by bool(BlendPose_0=@n6, BlendPose_1=@n8, bActiveValue=@n9) {blendType=Linear}
     bActiveValue=Get Is Wearing Backbling()
    ) {blendType=Linear}
    bActiveValue=Get bIsInFrontEnd()
   ) {blendType=Linear}
   n13 = RigidBody(ComponentPose=Local To Component(LocalPose=Use cached pose 'Pre sim'()), OverrideWorldGravity=Get GravityOverride()) {
    bClampLinearTranslationLimitToRefPose=true
    bOverrideWorldGravity=true
    overridePhysicsAsset=/BRCosmetics/Characters/Player/Male/Medium/Bodies/M_MED_CombCraterHat/Meshes/Parts/M_MED_CombCraterHat_Backpack_FaceAcc_Physics
    simulationSpace=BaseBoneSpace
    alphaBoolBlend {blendOption=Linear}
    baseBoneRef {boneName=neck_02}
    overrideWorldGravity {z=-980}
    simSpaceSettings {maxAngularAcceleration=65, maxAngularVelocity=15, maxLinearAcceleration=1000, maxLinearVelocity=3, velocityScaleZ=0.1, worldAlpha=0.3}
   }
   n14 = Component To Local(ComponentPose=@n13)
   n15 = Get Is Wearing Jacket() -> (Is Wearing Jacket:bool)
   n16 = Get Dyn_Alpha_Constr3() -> (Dyn_Alpha_Constr3:float)
   n17 = "Constraint - dyn_hair_fr_l_03 (P)"(
    ComponentPose=Local To Component(LocalPose=Blend Poses by bool(BlendPose_0=@n12, BlendPose_1=@n14, bActiveValue=@n15) {blendType=Linear})
    Alpha=@n16
    ConstraintWeights_0=1
   ) {
    alphaBoolBlend {blendOption=Linear}
    boneToModify {boneName=dyn_hair_fr_l_03}
    constraintSetup [{targetBone {boneName=spine_05}}]
    constraintWeights [1]
   }
   n18 = "Constraint - dyn_hair_fr_r_03 (P)"(ComponentPose=@n17, Alpha=@n16, ConstraintWeights_0=1) {
    alphaBoolBlend {blendOption=Linear}
    boneToModify {boneName=dyn_hair_fr_r_03}
    constraintSetup [{targetBone {boneName=spine_05}}]
    constraintWeights [1]
   }
   n19 = Get Dyn_Alpha_Constr2() -> (Dyn_Alpha_Constr2:float)
   n20 = "Constraint - dyn_hair_fr_l_02 (P)"(ComponentPose=@n18, Alpha=@n19, ConstraintWeights_0=1) {
    alphaBoolBlend {blendOption=Linear}
    boneToModify {boneName=dyn_hair_fr_l_02}
    constraintSetup [{targetBone {boneName=spine_05}}]
    constraintWeights [1]
   }
   n21 = "Constraint - dyn_hair_fr_r_02 (P)"(ComponentPose=@n20, Alpha=@n19, ConstraintWeights_0=1) {
    alphaBoolBlend {blendOption=Linear}
    boneToModify {boneName=dyn_hair_fr_r_02}
    constraintSetup [{targetBone {boneName=spine_05}}]
    constraintWeights [1]
   }
   n22 = Get bIsSkydiveFloating()
   n23 = "Transform (Modify) Bone - Bone: hair_fr_l_base"!(ComponentPose=@n21, bAlphaBoolEnabled:bool=@n22, Translation="0,0,0", Rotation="0,0,0", Scale="1,1,1") -> (Pose) {
    alphaInputType=Bool
    rotationMode=BMM_Additive
    translationMode=BMM_Additive
    alphaBoolBlend {blendOption=Linear}
    boneToModify {boneName=hair_fr_l_base}
   }
   n24 = "Transform (Modify) Bone - Bone: hair_fr_r_base"!(ComponentPose=@n23, bAlphaBoolEnabled:bool=@n22, Translation="0,0,0", Rotation="0,0,0", Scale="1,1,1") -> (Pose) {
    alphaInputType=Bool
    rotationMode=BMM_Additive
    translationMode=BMM_Additive
    alphaBoolBlend {blendOption=Linear}
    boneToModify {boneName=hair_fr_r_base}
   }
   n25 = Get parachuting() -> (parachuting:bool)
   n26 = "Transform (Modify) Bone - Bone: hair_fr_l_base"(ComponentPose=@n24, bAlphaBoolEnabled=@n25, Translation="1.125,0.000,0.000", Rotation="14.591,9.314,-27.503", Scale="1,1,1") {
    alphaInputType=Bool
    rotationMode=BMM_Additive
    translationMode=BMM_Additive
    alphaBoolBlend {blendOption=Linear}
    boneToModify {boneName=hair_fr_l_base}
    rotation=<rot 14.591 9.314 -27.503>
    translation {x=1.125}
   }
   n27 = "Transform (Modify) Bone - Bone: hair_fr_r_base"(ComponentPose=@n26, bAlphaBoolEnabled=@n25, Translation="0,0,0", Rotation="-13,-14.144,-31.401", Scale="1,1,1") {
    alphaInputType=Bool
    rotationMode=BMM_Additive
    translationMode=BMM_Additive
    alphaBoolBlend {blendOption=Linear}
    boneToModify {boneName=hair_fr_r_base}
    rotation=<rot -13 -14.144 -31.401>
   }
   Output Pose(Result=Component To Local(ComponentPose=@n27)) {layerGroup=None}
   n28 = Copy Pose From Mesh() {bCopyCurves=true, bUseAttachedParent=true}
   n29 = Copy Pose From Mesh(SourceMeshComponent=Get Body_Mesh() -> (Body_Mesh:Skeletal Mesh Component Budgeted Object Reference)) {bUseAttachedParent=true}
   n30 = "Constraint - hair_fr_r_base (PP)"!(
    ComponentPose=Local To Component(LocalPose=Layered blend per bone(BasePose=@n28, BlendPoses_0=@n29) {layerSetup [{branchFilters [{boneName=clavicle_l}, {boneName=clavicle_r}]}]})
    Alpha=1
    ConstraintWeights_0=1
    ConstraintWeights_1:float=0
   ) -> (Pose) {
    alphaBoolBlend {blendOption=Linear}
    boneToModify {boneName=hair_fr_r_base}
    constraintSetup [{targetBone {boneName=head}}, {targetBone {boneName=clavicle_r}}]
    constraintWeights [1,1]
   }
   n31 = "Constraint - hair_fr_l_base (PP)"!(ComponentPose=@n30, Alpha=1, ConstraintWeights_0=1, ConstraintWeights_1:float=0) -> (Pose) {
    alphaBoolBlend {blendOption=Linear}
    boneToModify {boneName=hair_fr_l_base}
    constraintSetup [{targetBone {boneName=head}}, {targetBone {boneName=clavicle_l}}]
    constraintWeights [1,1]
   }
   n32 = "Transform (Modify) Bone - Bone: dyn_hair_bck_l_base"(ComponentPose=@n31, Alpha=1, Translation="0.000,1.603,0.000", Rotation="27.909,33.087,-17.474", Scale="1,1,1") {
    rotationMode=BMM_Additive
    translationMode=BMM_Additive
    alphaBoolBlend {blendOption=Linear}
    boneToModify {boneName=dyn_hair_bck_l_base}
    rotation=<rot 27.909 33.087 -17.474>
    translation {y=1.603}
   }
   n33 = "Transform (Modify) Bone - Bone: dyn_hair_bck_r_base"(ComponentPose=@n32, Alpha=1, Translation="1.314,1.897,0.000", Rotation="40.830,52.480,6.738", Scale="1,1,1") {
    rotationMode=BMM_Additive
    translationMode=BMM_Additive
    alphaBoolBlend {blendOption=Linear}
    boneToModify {boneName=dyn_hair_bck_r_base}
    rotation=<rot 40.83 52.48 6.738>
    translation=<1.314 1.897>
   }
   n34 = "Transform (Modify) Bone - Bone: dyn_hair_side_l_base"(ComponentPose=@n33, Alpha=1, Translation="0,0,0", Rotation="-19.449,6.354,-37.598", Scale="1,1,1") {
    rotationMode=BMM_Additive
    translationMode=BMM_Additive
    alphaBoolBlend {blendOption=Linear}
    boneToModify {boneName=dyn_hair_side_l_base}
    rotation=<rot -19.449 6.354 -37.598>
   }
   n35 = "Transform (Modify) Bone - Bone: dyn_hair_side_r_base"(ComponentPose=@n34, Alpha=1, Translation="0,0,0", Rotation="-11.369,24.336,5.228", Scale="1,1,1") {
    rotationMode=BMM_Additive
    translationMode=BMM_Additive
    alphaBoolBlend {blendOption=Linear}
    boneToModify {boneName=dyn_hair_side_r_base}
    rotation=<rot -11.369 24.336 5.228>
   }
   n36 = "Transform (Modify) Bone - Bone: hair_fr_l_base"!(ComponentPose=@n35, Alpha:float=1, Translation="0,0,0", Rotation="-10.258,1.304,-18.17", Scale="1,1,1") -> (Pose) {
    rotationMode=BMM_Additive
    translationMode=BMM_Additive
    alphaBoolBlend {blendOption=Linear}
    boneToModify {boneName=hair_fr_l_base}
    rotation=<rot -10.258 1.304 -18.17>
   }
   n37 = "Transform (Modify) Bone - Bone: hair_fr_r_base"!(ComponentPose=@n36, Alpha:float=1, Translation="0,0,0", Rotation="-0.416,87.763,0.016", Scale="1,1,1") -> (Pose) {
    rotationMode=BMM_Additive
    translationMode=BMM_Additive
    alphaBoolBlend {blendOption=Linear}
    boneToModify {boneName=hair_fr_r_base}
    rotation=<rot -0.416 87.763 0.016>
   }
   n38 = "Transform (Modify) Bone - Bone: dyn_hair_bck_l_base"(ComponentPose=@n37, Alpha=1, Translation="-3.704,-1.834,0.000", Rotation="-26.264,-45.994,21.586", Scale="1,1,1") {
    rotationMode=BMM_Additive
    translationMode=BMM_Additive
    alphaBoolBlend {blendOption=Linear}
    boneToModify {boneName=dyn_hair_bck_l_base}
    rotation=<rot -26.264 -45.994 21.586>
    translation=<-3.704 -1.834>
   }
   n39 = "Transform (Modify) Bone - Bone: dyn_hair_bck_r_base"(ComponentPose=@n38, Alpha=1, Translation="-2.966,0.688,0.000", Rotation="-11.075,-58.504,11.431", Scale="1,1,1") {
    rotationMode=BMM_Additive
    translationMode=BMM_Additive
    alphaBoolBlend {blendOption=Linear}
    boneToModify {boneName=dyn_hair_bck_r_base}
    rotation=<rot -11.075 -58.504 11.431>
    translation=<-2.966 0.688>
   }
   n40 = "Transform (Modify) Bone - Bone: dyn_hair_side_l_base"(ComponentPose=@n39, Alpha=1, Translation="-1.096,-6.863,-2.228", Rotation="-9.761,-66.772,43.276", Scale="1,1,1") {
    rotationMode=BMM_Additive
    translationMode=BMM_Additive
    alphaBoolBlend {blendOption=Linear}
    boneToModify {boneName=dyn_hair_side_l_base}
    rotation=<rot -9.761 -66.772 43.276>
    translation=<-1.096 -6.863 -2.228>
   }
   n41 = "Transform (Modify) Bone - Bone: dyn_hair_side_r_base"(ComponentPose=@n40, Alpha=1, Translation="0,0,0", Rotation="-11.954,-63.675,-1.691", Scale="1,1,1") {
    rotationMode=BMM_Additive
    translationMode=BMM_Additive
    alphaBoolBlend {blendOption=Linear}
    boneToModify {boneName=dyn_hair_side_r_base}
    rotation=<rot -11.954 -63.675 -1.691>
   }
   n42 = "Transform (Modify) Bone - Bone: hair_fr_l_base"(ComponentPose=@n41, Alpha=1, Translation="0,0,0", Rotation="-90,30,-30", Scale="1,1,1") {
    rotationMode=BMM_Additive
    rotationSpace=BCS_BoneSpace
    translationMode=BMM_Additive
    translationSpace=BCS_BoneSpace
    alphaBoolBlend {blendOption=Linear}
    boneToModify {boneName=hair_fr_l_base}
    rotation=<rot -90 30 -30>
   }
   n43 = "Transform (Modify) Bone - Bone: hair_fr_r_base"(ComponentPose=@n42, Alpha=1, Translation="10,-2,0", Rotation="-20,-30,-60", Scale="1,1,1") {
    rotationMode=BMM_Additive
    rotationSpace=BCS_BoneSpace
    translationMode=BMM_Additive
    translationSpace=BCS_BoneSpace
    alphaBoolBlend {blendOption=Linear}
    boneToModify {boneName=hair_fr_r_base}
    rotation=<rot -20 -30 -60>
    translation=<10 -2>
   }
   n44 = "Transform (Modify) Bone - Bone: dyn_hair_bck_l_base"(ComponentPose=@n43, Alpha=1, Translation="0.000,-0.000,3.046", Rotation="0,0,110", Scale="1,1,1") {
    rotationMode=BMM_Additive
    translationMode=BMM_Additive
    alphaBoolBlend {blendOption=Linear}
    boneToModify {boneName=dyn_hair_bck_l_base}
    rotation {roll=110}
    translation {z=3.046}
   }
   n45 = "Transform (Modify) Bone - Bone: dyn_hair_bck_r_base"(ComponentPose=@n44, Alpha=1, Translation="0,0,6", Rotation="3.049,15.045,98.820", Scale="1,1,1") {
    rotationMode=BMM_Additive
    translationMode=BMM_Additive
    alphaBoolBlend {blendOption=Linear}
    boneToModify {boneName=dyn_hair_bck_r_base}
    rotation=<rot 3.049 15.045 98.82>
    translation {z=6}
   }
   n46 = "Transform (Modify) Bone - Bone: dyn_hair_side_l_base"(ComponentPose=@n45, Alpha=1, Translation="-0.000,0.000,1.519", Rotation="3.787,11.003,121.481", Scale="1,1,1") {
    rotationMode=BMM_Additive
    translationMode=BMM_Additive
    alphaBoolBlend {blendOption=Linear}
    boneToModify {boneName=dyn_hair_side_l_base}
    rotation=<rot 3.787 11.003 121.481>
    translation {z=1.519}
   }
   n47 = "Transform (Modify) Bone - Bone: dyn_hair_side_r_base"(ComponentPose=@n46, Alpha=1, Translation="-1.311,1.245,0.626", Rotation="-2.314,17.549,111.840", Scale="1,1,1") {
    rotationMode=BMM_Additive
    translationMode=BMM_Additive
    alphaBoolBlend {blendOption=Linear}
    boneToModify {boneName=dyn_hair_side_r_base}
    rotation=<rot -2.314 17.549 111.84>
    translation=<-1.311 1.245 0.626>
   }
   n48 = "Transform (Modify) Bone - Bone: hair_fr_l_base"(ComponentPose=@n47, Alpha=1, Translation="0,10,5", Rotation="0,0,10", Scale="1,1,1") {
    rotationMode=BMM_Additive
    translationMode=BMM_Additive
    alphaBoolBlend {blendOption=Linear}
    boneToModify {boneName=hair_fr_l_base}
    rotation {roll=10}
    translation {y=10, z=5}
   }
   n49 = "Transform (Modify) Bone - Bone: hair_fr_r_base"(ComponentPose=@n48, Alpha=1, Translation="0,0,0", Rotation="0,0,10", Scale="1,1,1") {
    rotationMode=BMM_Additive
    translationMode=BMM_Additive
    alphaBoolBlend {blendOption=Linear}
    boneToModify {boneName=hair_fr_r_base}
    rotation {roll=10}
   }
   Save cached pose 'Pre sim'(Pose=Component To Local(ComponentPose=@n49)) {cacheName=Pre sim}
   Get GravityOverride()
   Get Alpha() -> (Alpha:float)
   Get front hair alpfa() -> (front hair alpfa:double)
  ]}
 EventGraph {nodes [
   Event BlueprintUpdateAnimation()
   > n1 = Sequence!(execute) -> (then_0, then_1, then_2:Exec, then_3:Exec)
   n2 = Get bIsCrouching()
   n3 = Branch(execute=@n1.then_1, Condition=AND Boolean(A=Get bIsTargeting(), B=NOT Boolean(A=@n2)))
   > Set targeting float(targeting float:double=1) -> (Output_Get:double)
   Set targeting float(execute=@n3.else, targeting float=0)
   n4 = GetOwnerSkeletalMeshComponent()
   Set Head_Rot_Bk(
    execute=@n1.then_2
    Head_Rot_Bk:double=MapRangeClamped(
     Value=GetDeltaTransformFromRefPose!(self=@n4, BoneName=head, BaseName=spine_05) -> (
      ReturnValue_Location:Vector
      ReturnValue_Rotation_Roll:float
      ReturnValue_Rotation_Pitch:float
      ReturnValue_Rotation_Yaw:float
      ReturnValue_Scale:Vector
     ).ReturnValue_Rotation_Yaw
     InRangeA=-10
     InRangeB=-90
    )
   ) -> (Output_Get:double)
   n5 = Get bIsInFrontEnd()
   > n6 = Branch(Condition=@n5)
   n7 = MapRangeClamped(
    Value=GetDeltaTransformFromRefPose(self=GetOwnerSkeletalMeshComponent(), BoneName=head, BaseName=clavicle_l).ReturnValue_Rotation_Roll
    InRangeA=-20
    InRangeB=-90
   )
   > n8 = Set Head_Rot_right(Head_Rot_right:double=float * float(A=float * float(A=0.33, B:bool=@n5), B=@n7) {is=title K2Node_PromotableOperator}) -> (Output_Get:double)
   n9 = GetOwnerSkeletalMeshComponent()
   Set Head Rot left(
    execute=[@n8.then, @n11.then]
    Head Rot left:double=MapRangeClamped(Value=GetDeltaTransformFromRefPose(self=@n9, BoneName=head, BaseName=clavicle_r).ReturnValue_Rotation_Roll, InRangeA=0, InRangeB=90)
   ) -> (Output_Get:double)
   > Set Hat Sim Space(Hat Sim Space:AnimPhysSimSpaceType Enum=BoneRelative) -> (Output_Get:AnimPhysSimSpaceType Enum)
   > n10 = Branch(Condition=Get bIsSkydiveFloating())
   > Set front hair alpfa(front hair alpfa:double=0.3) -> (Output_Get:double)
   Set front hair alpfa(execute=@n10.else, front hair alpfa=1)
   n11 = Set Head_Rot_right(execute=@n6.else, Head_Rot_right=@n7)
   n12 = Branch(execute=@n1.then_3, Condition=Get bIsInFrontEnd())
   > Set Dyn_Alpha_Constr3(Dyn_Alpha_Constr3:float=0.66) -> (Output_Get:float)
   > Set Dyn_Alpha_Constr2(Dyn_Alpha_Constr2:float=0.1) -> (Output_Get:float)
   Set Dyn_Alpha_Constr3(execute=@n12.else, Dyn_Alpha_Constr3=0.1)
   > Set Dyn_Alpha_Constr2(Dyn_Alpha_Constr2=0)
   n13 = Gravity Override State Function DEPRECATED(
    execute=@n1.then_0
    CharacterPartAnimInstance=Self-Reference()
    GameTimeInSeconds=Get Delta Time X()
    GravityOverrideParams=Get Gravity Override Params()
   )
   > Set GravityOverride(GravityOverride=@n13)
   K2Node_Event_1 = CustomEvent!() -> (
    OutputDelegate
    then
    CustomBaseMesh:Skeletal Mesh Component Object Reference
    CustomParts:Array of Custom Character Part Object References
    CustomMeshParts:Array of Skeletal Mesh Component Budgeted Object References
   ) {is=title K2Node_CustomEvent}
   Set Is Wearing Jacket(
    execute=[@K2Node_Event_1.then, @n19.then]
    Is Wearing Jacket:bool=CharacterItemDefinitionHasAnyMetaTag(Tags=Make Array("[0]"=(TagName="Cosmetics.MetaTags.CombCraterJacket")))
   ) -> (Output_Get:bool)
   > Set Custom Base Mesh(Custom Base Mesh:Skeletal Mesh Component Object Reference=@K2Node_Event_1.CustomBaseMesh) -> (Output_Get:Skeletal Mesh Component Object Reference)
   n14 = TryGetPawnOwner()
   > n15 = Is Valid(InputObject=@n14)
   n16 = Set PawnOwner(execute=@n15.Is Not Valid, PawnOwner:Pawn Object Reference=TryGetFortPawn()) -> (Output_Get:Pawn Object Reference)
   Cast To FortPlayerPawn(execute=[@n16.then, @n18.then], Object=TryGetFortPawn())
   n17 = CheckForBackbling()
   > Set Is Wearing Backbling(Is Wearing Backbling=OR Boolean(A=@n17, B=@n17.bIsBackblingACape))
   > Branch(Condition=Get bIsInFrontEnd())
   n18 = Set PawnOwner(execute=@n15.Is Valid, PawnOwner=@n14)
   n19 = Event BlueprintInitializeAnimation()
   n20 = Event OnCharacterCustomizationLoaded()
   > Set Body_Mesh(Body_Mesh:Skeletal Mesh Component Budgeted Object Reference=Get Array Item with EFort Custom Part Type(Array=@n20.CustomMeshParts)) -> (Output_Get:Skeletal Mesh Component Budgeted Object Reference)
   Event OnParachuteOpened()
   > Set parachuting(parachuting:bool=true) -> (Output_Get:bool)
   Event OnParachuteClosed()
   > Set parachuting(parachuting=false)
   CheckForBackbling()
   MapRangeClamped(Value=0, InRangeA=10, InRangeB=90)
   Event OnCharacterAndPickaxeCustomizationLoaded!() -> (
    OutputDelegate
    then
    CustomBaseMesh:Skeletal Mesh Component Object Reference
    CustomParts:Array of Custom Character Part Object References
    CustomMeshParts:Array of Skeletal Mesh Component Budgeted Object References
   ) {is=title K2Node_Event}
  ]}
]
variables [
 Alpha {type=number, default=1}
 Body_Mesh {type=SkeletalMeshComponentBudgeted}
 Custom Base Mesh {type=SkeletalMeshComponent}
 Delta Time X {type=number}
 Dyn_Alpha_Constr2 {type=number, default=0.25}
 Dyn_Alpha_Constr3 {type=number, default=0.25}
 front hair alpfa {type=number}
 Gravity Override Params {
  type=GravityOverrideParamsStruct
  default {
   skydive_DivingDown {gravityOverride=<-1500 500 0>}
   skydive_FloatingDown {gravityOverride=<0 -700 -500>}
   skydive_GoingUp {gravityOverride=<0 -300 -200>}
  }
 }
 GravityOverride {type=Vector}
 Hat Sim Space {type=AnimPhysSimSpaceType}
 Head Rot left {type=number}
 Head_Rot_Bk {type=number}
 Head_Rot_right {type=number}
 Is Wearing Backbling {type=boolean}
 Is Wearing Cape {type=boolean}
 Is Wearing Jacket {type=boolean}
 parachuting {type=boolean}
 PawnOwner {type=Pawn, default=null}
 targeting float {type=number}
]
